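; QA helper Blueprint that logs the world locations of every actor of a chosen class as BugItGo commands.
; A custom GetActorLocations event gathers all actors of the class set in the GetActorClass variable.
; For each, it formats X, Y, Z without decimals or grouping, appends zero rotation, and prints the line through MusicPartyFunctionLibrary's PrintLog.
; The construction script is empty.
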(event Custom|GetActorLocations
  (bind _outactors (Actor|GetAllActorsOfClass (Variables|Default|GetActorClass)))
  (for _array_element _outactors
    ;[1 1 1] Build Bugitgo from the actor location
    (bind (_returnvalue_x _returnvalue_y _returnvalue_z) (Math|Vector|BreakVector (Transformation|GetActorLocation _array_element)))
    (Class|MusicPartyFunctionLibrary|PrintLog (Utilities|Text|ToText(String) (Utilities|String|Append "BugItGo " (Utilities|String|ToString(Text) (Utilities|Text|ToText(Float) _returnvalue_x "HalfToEven" false false 1 324 0 0)) " " (Utilities|String|ToString(Text) (Utilities|Text|ToText(Float) _returnvalue_y "HalfToEven" false false 1 324 0 0)) " " (Utilities|String|ToString(Text) (Utilities|Text|ToText(Float) _returnvalue_z "HalfToEven" false false 1 324 0 0)) " " (Utilities|String|ToString(Text) (Utilities|Text|ToText(Float) 0.0 "HalfToEven" false false)) " " (Utilities|String|ToString(Text) (Utilities|Text|ToText(Float) 0.0 "HalfToEven" false false)) " " (Utilities|String|ToString(Text) (Utilities|Text|ToText(Float) 0.0 "HalfToEven" false false)))))))

(fn ConstructionScript)
