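; QA test Blueprint that lists the building level instances placed in a map.
; A custom "GetAll LevelInstances" event gathers every BuildingLevelInstance actor with GetAllActorsOfClass and prints each one in a loop for debugging.
; The construction script is empty.
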(event Custom|GetAll%20LevelInstances
  (bind _outactors (Actor|GetAllActorsOfClass "/Script/FortniteGame.BuildingLevelInstance"))
  (for _array_element _outactors
    (Debugging|PrintObject _array_element)))

(fn ConstructionScript)
